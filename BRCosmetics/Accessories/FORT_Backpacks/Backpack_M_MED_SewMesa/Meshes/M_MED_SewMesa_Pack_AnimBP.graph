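# Anim Blueprint for the SewMesa backpack, adding secondary motion to its plush-like head, ears, arms, legs, badge and strings.
# The AnimGraph chains additive pack bone transforms into root-relative AnimDynamics nodes, with mutual spherical limits separating string tips.
# The EventGraph derives smoothed pack rotation and offset from owner spine bend, speed, crouch, skydive and DBNO state.
# It also sets gravity, simulation scales and a body-type position offset.
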AnimBlueprint /BRCosmetics/Accessories/FORT_Backpacks/Backpack_M_MED_SewMesa/Meshes/M_MED_SewMesa_Pack_AnimBP
target_skeleton=&/M_MED_SewMesa_Pack_Skeleton
graphs [
 AnimGraph {nodes [
   n1 = "Transform (Modify) Bone - Bone: pack"(ComponentPose=Local To Component(LocalPose=Input Pose()), Translation="0,-5,0", Rotation="0,0,45") {
    rotationMode=BMM_Additive
    rotationSpace=BCS_BoneSpace
    translationMode=BMM_Additive
    translationSpace=BCS_BoneSpace
    alphaBoolBlend {blendOption=Linear}
    alphaScaleBiasClamp {interpSpeedDecreasing=5, interpSpeedIncreasing=5}
    boneToModify {boneName=pack}
    rotation {roll=45}
    translation {y=-5}
   }
   n2 = "Transform (Modify) Bone - Bone: pack"(ComponentPose=@n1, Translation="0,-1.5,0", Rotation="0,0,0") {
    rotationMode=BMM_Additive
    rotationSpace=BCS_BoneSpace
    translationMode=BMM_Additive
    translationSpace=BCS_BoneSpace
    alphaBoolBlend {blendOption=Linear}
    alphaScaleBiasClamp {interpSpeedDecreasing=5, interpSpeedIncreasing=5}
    boneToModify {boneName=pack}
    translation {y=-1.5}
   }
   n3 = "Transform (Modify) Bone - Bone: pack"(ComponentPose=@n2, Translation="0,0,0", Rotation="0,0,0") {rotationMode=BMM_Additive, translationMode=BMM_Additive, alphaBoolBlend {blendOption=Linear}, boneToModify {boneName=pack}}
   n4 = Anim Dynamics - Bone: dyn_head_01!(
    SimulationSpace:AnimPhysSimSpaceType Enum
    ComponentPose=@n3
    Alpha=1
    GravityOverride:Vector="0,0,-980"
    ComponentLinearAccScale:Vector="0.5,0.5,0.5"
    ComponentLinearVelScale:Vector="0.5,0.5,0.5"
   ) -> (Pose) {
    angularDampingOverride=0.999
    angularSpringConstant=190
    bAngularSpring=true
    bOverrideAngularDamping=true
    bUseGravityOverride=true
    simulationSpace=RootRelative
    alphaBoolBlend {blendOption=Linear}
    boundBone {boneName=dyn_head_01}
    componentAppliedLinearAccClamp=<500 500 500>
    componentLinearAccScale=<0.5 0.5 0.5>
    componentLinearVelScale=<0.5 0.5 0.5>
    gravityOverride {z=-980}
    physicsBodyDefinitions [{
      boundBone {boneName=dyn_head_01}
      boxExtents=<14 14 14>
      constraintSetup {angularLimitsMax=<1 10 10>, angularLimitsMin=<-1 -10 -10>, angularTarget=<1 0 0>}
      localJointOffset=<0.7 0 0>
     }]
   }
   n5 = Anim Dynamics - Bone: dyn_ear_l_01!(
    SimulationSpace:AnimPhysSimSpaceType Enum
    ComponentPose=@n4
    Alpha=1
    GravityOverride:Vector="0,0,-980"
    ComponentLinearAccScale:Vector="0.5,0.5,0.5"
    ComponentLinearVelScale:Vector="0.5,0.5,0.5"
   ) -> (Pose) {
    angularDampingOverride=0.99
    angularSpringConstant=150
    bAngularSpring=true
    bOverrideAngularDamping=true
    bUseGravityOverride=true
    simulationSpace=RootRelative
    alphaBoolBlend {blendOption=Linear}
    boundBone {boneName=dyn_ear_l_01}
    componentAppliedLinearAccClamp=<5000 5000 5000>
    componentLinearAccScale=<0.5 0.5 0.5>
    componentLinearVelScale=<0.5 0.5 0.5>
    gravityOverride {z=-980}
    physicsBodyDefinitions [{
      boundBone {boneName=dyn_ear_l_01}
      boxExtents=<5 5 5>
      constraintSetup {angularLimitsMax=<0 0 10>, angularLimitsMin=<0 0 -10>, angularTarget=<1 0 0>}
      localJointOffset=<0.05 0 0>
     }]
   }
   n6 = Anim Dynamics - Bone: dyn_ear_r_01!(
    SimulationSpace:AnimPhysSimSpaceType Enum
    ComponentPose=@n5
    Alpha=1
    GravityOverride:Vector="0,0,-980"
    ComponentLinearAccScale:Vector="0.5,0.5,0.5"
    ComponentLinearVelScale:Vector="0.5,0.5,0.5"
   ) -> (Pose) {
    angularDampingOverride=0.99
    angularSpringConstant=150
    bAngularSpring=true
    bOverrideAngularDamping=true
    bUseGravityOverride=true
    simulationSpace=RootRelative
    alphaBoolBlend {blendOption=Linear}
    boundBone {boneName=dyn_ear_r_01}
    componentAppliedLinearAccClamp=<5000 5000 5000>
    componentLinearAccScale=<0.5 0.5 0.5>
    componentLinearVelScale=<0.5 0.5 0.5>
    gravityOverride {z=-980}
    physicsBodyDefinitions [{
      boundBone {boneName=dyn_ear_r_01}
      boxExtents=<5 5 5>
      constraintSetup {angularLimitsMax=<0 0 10>, angularLimitsMin=<0 0 -10>, angularTarget=<1 0 0>}
      localJointOffset=<0.05 0 0>
     }]
   }
   n7 = Anim Dynamics - Bone: dyn_badge_01!(
    SimulationSpace:AnimPhysSimSpaceType Enum
    ComponentPose=@n6
    Alpha=1
    ComponentLinearAccScale:Vector="0.5,0.5,0.5"
    ComponentLinearVelScale:Vector="0.5,0.5,0.5"
   ) -> (Pose) {
    simulationSpace=RootRelative
    alphaBoolBlend {blendOption=Linear}
    boundBone {boneName=dyn_badge_01}
    componentAppliedLinearAccClamp=<5000 5000 5000>
    componentLinearAccScale=<0.5 0.5 0.5>
    componentLinearVelScale=<0.5 0.5 0.5>
    physicsBodyDefinitions [{
      boundBone {boneName=dyn_badge_01}
      boxExtents=<3 3 3>
      constraintSetup {angularLimitsMax=<0 0 5>, angularLimitsMin=<0 0 -5>, angularTarget=<1 0 0>}
      localJointOffset=<0.05 0 0>
     }]
   }
   n8 = Anim Dynamics - Bone: dyn_string_r_01!(
    SimulationSpace:AnimPhysSimSpaceType Enum
    ComponentPose=@n7
    Alpha=1
    GravityOverride:Vector="0,0,-980"
    ComponentLinearAccScale:Vector="0.8,0.8,0.8"
    ComponentLinearVelScale:Vector="0.8,0.8,0.8"
   ) -> (Pose) {
    angularDampingOverride=0.999
    angularSpringConstant=180
    bAngularSpring=true
    bOverrideAngularDamping=true
    bUseGravityOverride=true
    simulationSpace=RootRelative
    alphaBoolBlend {blendOption=Linear}
    boundBone {boneName=dyn_string_r_01}
    componentAppliedLinearAccClamp=<5500 5500 5500>
    componentLinearAccScale=<0.8 0.8 0.8>
    componentLinearVelScale=<0.8 0.8 0.8>
    gravityOverride {z=-980}
    physicsBodyDefinitions [{
      boundBone {boneName=dyn_string_r_01}
      boxExtents=<4 4 4>
      constraintSetup {angularLimitsMax=<1 0 18>, angularLimitsMin=<-1 -50 -18>, angularTarget=<1 0 0>}
      localJointOffset=<0.09 0 0>
     }]
   }
   n9 = Anim Dynamics - Bone: dyn_string_r_02!(
    SimulationSpace:AnimPhysSimSpaceType Enum
    ComponentPose=@n8
    Alpha=1
    GravityOverride:Vector="0,0,-980"
    ComponentLinearAccScale:Vector="0.8,0.8,0.8"
    ComponentLinearVelScale:Vector="0.8,0.8,0.8"
   ) -> (Pose) {
    angularDampingOverride=0.999
    angularSpringConstant=180
    bAngularSpring=true
    bOverrideAngularDamping=true
    bUseGravityOverride=true
    bUsePlanarLimit=false
    simulationSpace=RootRelative
    alphaBoolBlend {blendOption=Linear}
    boundBone {boneName=dyn_string_r_02}
    componentAppliedLinearAccClamp=<5500 5500 5500>
    componentLinearAccScale=<0.8 0.8 0.8>
    componentLinearVelScale=<0.8 0.8 0.8>
    gravityOverride {z=-980}
    physicsBodyDefinitions [{
      boundBone {boneName=dyn_string_r_02}
      boxExtents=<4 4 4>
      constraintSetup {angularLimitsMax=<1 0 18>, angularLimitsMin=<-1 -50 -18>, angularTarget=<1 0 0>}
      localJointOffset=<0.05 0 0>
     }]
    sphericalLimits [{limitRadius=0.3, drivingBone {boneName=dyn_string_l_02}, sphereLocalOffset=<0 0 0>}]
   }
   n10 = Anim Dynamics - Bone: dyn_string_l_01!(
    SimulationSpace:AnimPhysSimSpaceType Enum
    ComponentPose=@n9
    Alpha=1
    GravityOverride:Vector="0,0,-980"
    ComponentLinearAccScale:Vector="1,1,1"
    ComponentLinearVelScale:Vector="1,1,1"
   ) -> (Pose) {
    angularDampingOverride=0.999
    angularSpringConstant=180
    bAngularSpring=true
    bOverrideAngularDamping=true
    bUseGravityOverride=true
    simulationSpace=RootRelative
    alphaBoolBlend {blendOption=Linear}
    boundBone {boneName=dyn_string_l_01}
    componentAppliedLinearAccClamp=<5500 5500 5500>
    componentLinearAccScale=<1 1 1>
    componentLinearVelScale=<1 1 1>
    gravityOverride {z=-980}
    physicsBodyDefinitions [{
      boundBone {boneName=dyn_string_l_01}
      boxExtents=<5 5 5>
      constraintSetup {angularLimitsMax=<1 0 18>, angularLimitsMin=<-1 -50 -18>, angularTarget=<1 0 0>}
      localJointOffset=<0.11 0 0>
     }]
   }
   n11 = Anim Dynamics - Bone: dyn_string_l_02!(
    SimulationSpace:AnimPhysSimSpaceType Enum
    ComponentPose=@n10
    Alpha=1
    GravityOverride:Vector="0,0,-980"
    ComponentLinearAccScale:Vector="1,1,1"
    ComponentLinearVelScale:Vector="1,1,1"
   ) -> (Pose) {
    angularDampingOverride=0.999
    angularSpringConstant=180
    bAngularSpring=true
    bOverrideAngularDamping=true
    bUseGravityOverride=true
    bUsePlanarLimit=false
    simulationSpace=RootRelative
    alphaBoolBlend {blendOption=Linear}
    boundBone {boneName=dyn_string_l_02}
    componentAppliedLinearAccClamp=<5500 5500 5500>
    componentLinearAccScale=<1 1 1>
    componentLinearVelScale=<1 1 1>
    gravityOverride {z=-980}
    physicsBodyDefinitions [{
      boundBone {boneName=dyn_string_l_02}
      boxExtents=<4 4 4>
      constraintSetup {angularLimitsMax=<1 0 18>, angularLimitsMin=<-1 -50 -18>, angularTarget=<1 0 0>}
      localJointOffset=<0.05 0 0>
     }]
    sphericalLimits [{limitRadius=0.3, drivingBone {boneName=dyn_string_r_02}, sphereLocalOffset=<0 0 0>}]
   }
   n12 = Anim Dynamics - Bone: dyn_arm_r_01!(
    SimulationSpace:AnimPhysSimSpaceType Enum
    ComponentPose=@n11
    Alpha=1
    GravityOverride:Vector="0,0,-980"
    ComponentLinearAccScale:Vector="0.45,0.45,0.45"
    ComponentLinearVelScale:Vector="0.45,0.45,0.45"
   ) -> (Pose) {
    angularDampingOverride=0.99
    angularSpringConstant=180
    bAngularSpring=true
    bOverrideAngularDamping=true
    bUseGravityOverride=true
    bUsePlanarLimit=false
    simulationSpace=RootRelative
    alphaBoolBlend {blendOption=Linear}
    boundBone {boneName=dyn_arm_r_01}
    componentAppliedLinearAccClamp=<3000 3000 3000>
    componentLinearAccScale=<0.45 0.45 0.45>
    componentLinearVelScale=<0.45 0.45 0.45>
    gravityOverride {z=-980}
    physicsBodyDefinitions [{
      boundBone {boneName=dyn_arm_r_01}
      boxExtents=<10 10 10>
      constraintSetup {angularLimitsMax=<0 0 20>, angularLimitsMin=<0 -20 0>, angularTarget=<1 0 0>}
      localJointOffset=<1 0 0>
     }]
   }
   n13 = Anim Dynamics - Bone: dyn_arm_r_02!(
    SimulationSpace:AnimPhysSimSpaceType Enum
    ComponentPose=@n12
    Alpha=1
    GravityOverride:Vector="0,0,-980"
    ComponentLinearAccScale:Vector="0.45,0.45,0.45"
    ComponentLinearVelScale:Vector="0.45,0.45,0.45"
   ) -> (Pose) {
    angularDampingOverride=0.99
    angularSpringConstant=180
    bAngularSpring=true
    bUseGravityOverride=true
    bUsePlanarLimit=false
    simulationSpace=RootRelative
    alphaBoolBlend {blendOption=Linear}
    boundBone {boneName=dyn_arm_r_02}
    componentAppliedLinearAccClamp=<3000 3000 3000>
    componentLinearAccScale=<0.45 0.45 0.45>
    componentLinearVelScale=<0.45 0.45 0.45>
    gravityOverride {z=-980}
    physicsBodyDefinitions [{
      boundBone {boneName=dyn_arm_r_02}
      boxExtents=<10 10 10>
      constraintSetup {angularLimitsMax=<0 0 20>, angularLimitsMin=<0 -20 0>, angularTarget=<1 0 0>}
      localJointOffset=<1 0 0>
     }]
   }
   n14 = Anim Dynamics - Bone: dyn_arm_l_01!(
    SimulationSpace:AnimPhysSimSpaceType Enum
    ComponentPose=@n13
    Alpha=1
    GravityOverride:Vector="0,0,-980"
    ComponentLinearAccScale:Vector="0.45,0.45,0.45"
    ComponentLinearVelScale:Vector="0.45,0.45,0.45"
   ) -> (Pose) {
    angularDampingOverride=0.99
    angularSpringConstant=180
    bAngularSpring=true
    bOverrideAngularDamping=true
    bUseGravityOverride=true
    bUsePlanarLimit=false
    simulationSpace=RootRelative
    alphaBoolBlend {blendOption=Linear}
    boundBone {boneName=dyn_arm_l_01}
    componentAppliedLinearAccClamp=<3000 3000 3000>
    componentLinearAccScale=<0.45 0.45 0.45>
    componentLinearVelScale=<0.45 0.45 0.45>
    gravityOverride {z=-980}
    physicsBodyDefinitions [{
      boundBone {boneName=dyn_arm_l_01}
      boxExtents=<10 10 10>
      constraintSetup {angularLimitsMax=<0 0 0>, angularLimitsMin=<0 -20 -20>, angularTarget=<1 0 0>}
      localJointOffset=<1 0 0>
     }]
   }
   n15 = Anim Dynamics - Bone: dyn_arm_l_02!(
    SimulationSpace:AnimPhysSimSpaceType Enum
    ComponentPose=@n14
    Alpha=1
    GravityOverride:Vector="0,0,-980"
    ComponentLinearAccScale:Vector="0.45,0.45,0.45"
    ComponentLinearVelScale:Vector="0.45,0.45,0.45"
   ) -> (Pose) {
    angularDampingOverride=0.99
    angularSpringConstant=180
    bAngularSpring=true
    bOverrideAngularDamping=true
    bUseGravityOverride=true
    bUsePlanarLimit=false
    simulationSpace=RootRelative
    alphaBoolBlend {blendOption=Linear}
    boundBone {boneName=dyn_arm_l_02}
    componentAppliedLinearAccClamp=<3000 3000 3000>
    componentLinearAccScale=<0.45 0.45 0.45>
    componentLinearVelScale=<0.45 0.45 0.45>
    gravityOverride {z=-980}
    physicsBodyDefinitions [{
      boundBone {boneName=dyn_arm_l_02}
      boxExtents=<10 10 10>
      constraintSetup {angularLimitsMax=<0 0 0>, angularLimitsMin=<0 -20 -20>, angularTarget=<1 0 0>}
      localJointOffset=<1 0 0>
     }]
   }
   n16 = Anim Dynamics - Bone: dyn_lowerBody_01!(
    SimulationSpace:AnimPhysSimSpaceType Enum
    ComponentPose=@n15
    Alpha=1
    GravityOverride:Vector="0,0,-980"
    ComponentLinearAccScale:Vector="0.5,0.5,0.5"
    ComponentLinearVelScale:Vector="0.5,0.5,0.5"
   ) -> (Pose) {
    angularDampingOverride=0.99
    bOverrideAngularDamping=true
    bUseGravityOverride=true
    simulationSpace=RootRelative
    alphaBoolBlend {blendOption=Linear}
    boundBone {boneName=dyn_lowerBody_01}
    componentAppliedLinearAccClamp=<5000 5000 5000>
    componentLinearAccScale=<0.5 0.5 0.5>
    componentLinearVelScale=<0.5 0.5 0.5>
    gravityOverride {z=-980}
    physicsBodyDefinitions [{
      boundBone {boneName=dyn_lowerBody_01}
      boxExtents=<15 15 15>
      constraintSetup {angularLimitsMax=<3 0 10>, angularLimitsMin=<-3 -5 -10>, angularTarget=<1 0 0>}
      localJointOffset=<0.05 0 0>
     }]
   }
   n17 = Anim Dynamics - Bone: dyn_leg_r_01!(
    SimulationSpace:AnimPhysSimSpaceType Enum
    ComponentPose=@n16
    Alpha=1
    GravityOverride:Vector="0,0,-980"
    ComponentLinearAccScale:Vector="0.5,0.5,0.5"
    ComponentLinearVelScale:Vector="0.5,0.5,0.5"
   ) -> (Pose) {
    angularDampingOverride=0.99
    angularSpringConstant=180
    bAngularSpring=true
    bOverrideAngularDamping=true
    bUseGravityOverride=true
    bUsePlanarLimit=false
    simulationSpace=RootRelative
    alphaBoolBlend {blendOption=Linear}
    boundBone {boneName=dyn_leg_r_01}
    componentAppliedLinearAccClamp=<3000 3000 3000>
    componentLinearAccScale=<0.5 0.5 0.5>
    componentLinearVelScale=<0.5 0.5 0.5>
    gravityOverride {z=-980}
    physicsBodyDefinitions [{
      boundBone {boneName=dyn_leg_r_01}
      boxExtents=<10 10 10>
      constraintSetup {angularLimitsMax=<0 0 15>, angularLimitsMin=<0 -15 -15>, angularTarget=<1 0 0>}
      localJointOffset=<0.5 0 0>
     }]
   }
   n18 = Anim Dynamics - Bone: dyn_leg_r_02!(
    SimulationSpace:AnimPhysSimSpaceType Enum
    ComponentPose=@n17
    Alpha=1
    GravityOverride:Vector="0,0,-980"
    ComponentLinearAccScale:Vector="0.5,0.5,0.5"
    ComponentLinearVelScale:Vector="0.5,0.5,0.5"
   ) -> (Pose) {
    angularDampingOverride=0.99
    angularSpringConstant=180
    bAngularSpring=true
    bOverrideAngularDamping=true
    bUseGravityOverride=true
    bUsePlanarLimit=false
    simulationSpace=RootRelative
    alphaBoolBlend {blendOption=Linear}
    boundBone {boneName=dyn_leg_r_02}
    componentAppliedLinearAccClamp=<3000 3000 3000>
    componentLinearAccScale=<0.5 0.5 0.5>
    componentLinearVelScale=<0.5 0.5 0.5>
    gravityOverride {z=-980}
    physicsBodyDefinitions [{
      boundBone {boneName=dyn_leg_r_02}
      boxExtents=<10 10 10>
      constraintSetup {angularLimitsMax=<0 0 15>, angularLimitsMin=<0 -15 -15>, angularTarget=<1 0 0>}
      localJointOffset=<0.5 0 0>
     }]
   }
   n19 = Anim Dynamics - Bone: dyn_leg_l_01!(
    SimulationSpace:AnimPhysSimSpaceType Enum
    ComponentPose=@n18
    Alpha=1
    GravityOverride:Vector="0,0,-980"
    ComponentLinearAccScale:Vector="0.5,0.5,0.5"
    ComponentLinearVelScale:Vector="0.5,0.5,0.5"
   ) -> (Pose) {
    angularDampingOverride=0.99
    angularSpringConstant=180
    bAngularSpring=true
    bOverrideAngularDamping=true
    bUseGravityOverride=true
    bUsePlanarLimit=false
    simulationSpace=RootRelative
    alphaBoolBlend {blendOption=Linear}
    boundBone {boneName=dyn_leg_l_01}
    componentAppliedLinearAccClamp=<3000 3000 3000>
    componentLinearAccScale=<0.5 0.5 0.5>
    componentLinearVelScale=<0.5 0.5 0.5>
    gravityOverride {z=-980}
    physicsBodyDefinitions [{
      boundBone {boneName=dyn_leg_l_01}
      boxExtents=<10 10 10>
      constraintSetup {angularLimitsMax=<0 0 15>, angularLimitsMin=<0 -15 -15>, angularTarget=<1 0 0>}
      localJointOffset=<0.5 0 0>
     }]
   }
   n20 = Anim Dynamics - Bone: dyn_leg_l_02!(
    SimulationSpace:AnimPhysSimSpaceType Enum
    ComponentPose=@n19
    Alpha=1
    GravityOverride:Vector="0,0,-980"
    ComponentLinearAccScale:Vector="0.5,0.5,0.5"
    ComponentLinearVelScale:Vector="0.5,0.5,0.5"
   ) -> (Pose) {
    angularDampingOverride=0.99
    angularSpringConstant=180
    bAngularSpring=true
    bOverrideAngularDamping=true
    bUseGravityOverride=true
    bUsePlanarLimit=false
    simulationSpace=RootRelative
    alphaBoolBlend {blendOption=Linear}
    boundBone {boneName=dyn_leg_l_02}
    componentAppliedLinearAccClamp=<3000 3000 3000>
    componentLinearAccScale=<0.5 0.5 0.5>
    componentLinearVelScale=<0.5 0.5 0.5>
    gravityOverride {z=-980}
    physicsBodyDefinitions [{
      boundBone {boneName=dyn_leg_l_02}
      boxExtents=<10 10 10>
      constraintSetup {angularLimitsMax=<0 0 15>, angularLimitsMin=<0 -15 -15>, angularTarget=<1 0 0>}
      localJointOffset=<0.5 0 0>
     }]
   }
   Output Pose(Result=Component To Local(ComponentPose=@n20)) {layerGroup=None}
  ]}
 EventGraph {nodes [
   n1 = Event BlueprintUpdateAnimation()
   > Set Delta Time X(Delta Time X=@n1.DeltaTimeX)
   > n2 = Sequence!(execute) -> (then_0, then_1, then_2:Exec)
   n3 = Branch(execute=@n2.then_1, Condition=Get bIsInFrontEnd())
   > Set SimlationSpace(SimlationSpace:AnimPhysSimSpaceType Enum=Component) -> (Output_Get:AnimPhysSimSpaceType Enum)
   > Set VelocityScAcc_A(VelocityScAcc_A:Vector="1,1,1") -> (Output_Get:Vector)
   > Set VelocityScAcc_B(VelocityScAcc_B:Vector="0.8,0.8,0.8") -> (Output_Get:Vector)
   > Set VelocityScAcc_C(VelocityScAcc_C:Vector="0.5,0.5,0.5") -> (Output_Get:Vector)
   > Set VelocityScAcc_D(VelocityScAcc_D:Vector="0.45,0.45,0.45") -> (Output_Get:Vector)
   Set SimlationSpace(execute=@n3.else)
   > Set VelocityScAcc_A(VelocityScAcc_A="1,1,1")
   > Set VelocityScAcc_B(VelocityScAcc_B="0.8,0.8,0.8")
   > Set VelocityScAcc_C(VelocityScAcc_C="0.5,0.5,0.5")
   > Set VelocityScAcc_D(VelocityScAcc_D="0.45,0.45,0.45")
   n4 = Branch(execute=@n2.then_0, Condition=Get bIsSkyDiving())
   > n5 = Gravity Override State Function DEPRECATED(
    CharacterPartAnimInstance=Self-Reference()
    GameTimeInSeconds=GetGameTimeInSeconds()
    GravityOverrideParams=Get Gravity Override Params()
    DebugJointName=spine_05
   )
   > Set Gravity Override(Gravity Override=@n5)
   Set Gravity Override(execute=@n4.else, Gravity Override="0,0,-980")
   n6 = Sequence!(execute=@n2.then_2) -> (then_0, then_1, then_2, then_3:Exec, then_4:Exec, then_5:Exec)
   > n7 = Branch(Condition=Get bIsDBNO())
   > Set packDBNOTransY(packDBNOTransY:double=-1.8) -> (Output_Get:double)
   Set packDBNOTransY(execute=@n7.else, packDBNOTransY=0)
   n8 = Branch(execute=@n6.then_5, Condition=Get bIsInFrontEnd())
   n9 = GetOwnerSkeletalMeshComponent()
   n10 = Set Head Down Rot(
    execute=@n8.else
    Head Down Rot:float=GetAlphaFromReferenceTransformDelta(SkeletalMeshComponent=@n9, BoneName=spine_01, BaseName=head, InputRange=(X=0,Y=-45), TwistAxis="1,0,0")
   ) -> (Output_Get:float)
   n11 = "To Float (Boolean)"(InBool=Get bIsPlayingMeleeAnim())
   n12 = Get Delta Time X()
   n13 = Get isMelee() -> (isMelee:double)
   > Set isMelee(isMelee:double=float * float(A=@n10, B=FInterpTo(Current=@n13, Target=@n11, DeltaTime=@n12, InterpSpeed=1)) {is=title K2Node_PromotableOperator}) -> (Output_Get:double)
   n14 = BreakRotator(InRot=BreakTransform(InTransform=GetDeltaTransformFromRefPose(self=@n9, BoneName=upperarm_l, BaseName=clavicle_l)).Rotation)
   > Set Backpack Pickaxe Rot(Backpack Pickaxe Rot:Rotator=MakeRotator(Roll=MapRangeClamped(Value=@n14.Yaw, InRangeA=0, InRangeB=90, OutRangeB=50))) -> (Output_Get:Rotator)
   n15 = Branch(execute=@n6.then_2, Condition=Get isM_Med() -> (isM_Med:bool))
   > n16 = Branch(Condition=Get bIsSkyDiving())
   n17 = MapRangeClamped(Value=Get Spine Delta Transform from Ref Z() -> (Spine Delta Transform from Ref Z:float), InRangeA=10, InRangeB=0, OutRangeB=20)
   > Set TargetRotX(TargetRotX:double=@n17) -> (Output_Get:double)
   n18 = Branch(execute=@n16.else, Condition=Get bIsCrouching())
   n19 = MapRangeClamped(Value=Get Spine Delta Transform from Ref Z(), InRangeA=8, InRangeB=-8, OutRangeB=50)
   n20 = MapRangeClamped(Value=Get Spine Delta Transform from Ref Z(), InRangeA=-8, InRangeB=-30, OutRangeB=15)
   n21 = MapRangeClamped(Value=Get Spine Delta Transform from Ref Z(), InRangeA=-30, InRangeB=-45, OutRangeB=90)
   > Set TargetRotX(TargetRotX=float + float(A=10, B=float + float!(A=@n17, B=@n19, C:double=@n20, D:double=@n21) -> (ReturnValue)))
   n22 = Vector Length(A=Get PawnVelocity())
   n23 = Branch(execute=@n18.else, Condition=AND Boolean(A="float > float"(A=@n22), B="float < float"(A=@n22, B=150)))
   n24 = Get bIsSkydiveFloating()
   n25 = "To Float (Boolean)"(InBool=AND Boolean(A=NOT Boolean(A=@n24), B=Get bIsSkyDiving()))
   n26 = Get Spine Delta Transform from Ref Z()
   n27 = Get bIsInFrontEnd()
   n28 = float + float!(
    A=MapRangeClamped(Value="To Float (Boolean)"(InBool=@n24), InRangeA=0, InRangeB=1, OutRangeB=-15)
    B=float * float(
     A=MapRangeClamped(Value=@n25, InRangeA=0, InRangeB=1, OutRangeB=-40)
     B=MapRangeClamped(Value=@n26, InRangeA=-20, InRangeB=-90, OutRangeA=1, OutRangeB=0)
    )
    C=float * float!(A=@n19, B:bool=NOT Boolean(A=@n27), C:bool=Get bIsSkyDiving()) -> (ReturnValue)
    D=@n20
    E:double=@n21
   ) -> (ReturnValue)
   n29 = float * float!(A=@n28, B:double=MapRangeClamped(Value=@n22, InRangeA=150, InRangeB=500)) -> (ReturnValue)
   > Set TargetRotX(TargetRotX=float + float!(A=@n29, B=28) -> (ReturnValue))
   n30 = Get Spine Delta Transform from Ref Z()
   n31 = AND Boolean!(A="float < float"(A=@n22, B=50), B="float > float"(A=@n22), C:bool="float < float"(A=@n30, B=10), D:bool="float > float"(A=@n30)) -> (ReturnValue)
   Set TargetRotX(execute=@n23.else, TargetRotX=float + float(A=@n29, B=float * float(A=5, B:bool=@n31)))
   Set TargetRotX(execute=@n15.else, TargetRotX=@n28)
   n32 = GetDeltaTransformFromRefPose(self=GetOwnerSkeletalMeshComponent(), BoneName=spine_05, BaseName=pelvis)
   Set Spine Delta Transform from Ref Z(execute=@n6.then_1, Spine Delta Transform from Ref Z:float=BreakRotator(InRot=BreakTransform(InTransform=@n32).Rotation).Yaw) -> (Output_Get:float)
   n33 = BreakRotator(InRot=GetDeltaTransformFromRefPose!(self=GetOwnerSkeletalMeshComponent(), BoneName=spine_05, BaseName=spine_01) -> (ReturnValue_Location:Vector, ReturnValue_Rotation:Rotator, ReturnValue_Scale:Vector).ReturnValue_Rotation)
   > Set Paraсhuting(Paraсhuting:double=float * float(A:bool=Get bIsSkyDiving(), B:double=MapRangeClamped(Value=@n33.Yaw, InRangeA=-10, InRangeB=-45))) -> (Output_Get:double)
   n34 = float * float(A:double=Get Delta Time X(), B=Get DeltaTimeModifier() -> (DeltaTimeModifier:double))
   n35 = Get TargetRotX() -> (TargetRotX:double)
   n36 = Get!() -> (BackpackRotate_Roll:float, BackpackRotate_Pitch:float, BackpackRotate_Yaw:float)
   Set BackpackRotate(execute=@n6.then_4, BackpackRotate:Rotator=MakeRotator(Roll=FInterpTo(Current=@n36, Target=@n35, DeltaTime=@n34, InterpSpeed=1))) -> (Output_Get:Rotator)
   n37 = OR Boolean(A=Get bIsSkyDiving(), B=Get bIsSkydiveFloating())
   n38 = float + float!(
    A=MapRangeClamped(Value=Get Spine Delta Transform from Ref Z(), InRangeA=-8, InRangeB=-30, OutRangeB=-3)
    B=Get packDBNOTransY() -> (packDBNOTransY:double)
    C:double=MapRangeClamped(Value=Get Spine Delta Transform from Ref Z(), InRangeA=-50, InRangeB=-75, OutRangeB=-6.5)
    D:double=MapRangeClamped(Value="To Float (Boolean)"(InBool=@n37), InRangeA=0, InRangeB=1, OutRangeB=-1.5)
    E:double=Get BacpackInitPositionAdd() -> (BacpackInitPositionAdd:double)
   ) -> (ReturnValue)
   n39 = Get!() -> (BackpackTranslation_X:double, BackpackTranslation_Y:double, BackpackTranslation_Z:double)
   > Set BackpackTranslation(BackpackTranslation:Vector=MakeVector(Y=FInterpTo(Current=@n39.BackpackTranslation_Y, Target=@n38, DeltaTime=@n34, InterpSpeed=5))) -> (Output_Get:Vector)
   n40 = Event OnCharacterCustomizationLoaded()
   > n41 = Sequence!(execute) -> (then_0, then_1)
   n42 = "Get (a copy)"(Array:Array of Custom Character Part Object References=@n40.CustomParts, Dimension 1=1) -> (Output:Custom Character Part Object Reference)
   n43 = Get GenderPermitted!(self:Custom Character Part Object Reference=@n42) -> (GenderPermitted:EFortCustomGender Enum=Invalid)
   n44 = Get BodyTypesPermitted!(self:Custom Character Part Object Reference=@n42) -> (BodyTypesPermitted:EFortCustomBodyType Enum=NONE)
   > Set isM_Med(isM_Med:bool=AND Boolean!(
     A="Equal (Enum)"(A:EFortCustomGender Enum=@n43, B:EFortCustomGender Enum=Male)
     B="Equal (Enum)"(A:EFortCustomBodyType Enum=@n44, B:EFortCustomBodyType Enum=Medium)
    ) -> (ReturnValue)) -> (Output_Get:bool)
   > n45 = Switch on EFortCustomGender!(execute, Selection:EFortCustomGender Enum=@n43) -> (Invalid:Exec, Male:Exec, Female:Exec, Both:Exec) {is=title K2Node_SwitchEnum}
   n46 = Switch on EFortCustomBodyType!(execute=@n45.Female, Selection:EFortCustomBodyType Enum=@n44) -> (Small:Exec, Medium:Exec, MediumAndSmall:Exec, Large:Exec, LargeAndSmall:Exec, LargeAndMedium:Exec, All:Exec, Deprecated:Exec) {is=title K2Node_SwitchEnum}
   Set bodyTypeNo(execute=@n46.Large, bodyTypeNo:int=5) -> (Output_Get:int=0)
   Set bodyTypeNo(execute=@n46.Medium, bodyTypeNo=4) -> (Output_Get=0)
   Set bodyTypeNo(execute=@n46.Small, bodyTypeNo=3) -> (Output_Get=0)
   n47 = Switch on EFortCustomBodyType(execute=@n45.Male, Selection=@n44)
   Set bodyTypeNo(execute=@n47.Large, bodyTypeNo=2) -> (Output_Get=0)
   Set bodyTypeNo(execute=@n47.Medium, bodyTypeNo=1) -> (Output_Get=0)
   Set bodyTypeNo(execute=@n47.Small, bodyTypeNo=0) -> (Output_Get=0)
   Set BacpackInitPositionAdd(
    execute=@n41.then_1
    BacpackInitPositionAdd:double=Select!(
     Option 0:double=0
     Option 1:double=2
     Option 2:double=0.6
     Option 3:double=0
     Option 4:double=3
     Option 5:double=0.8
     Index:int=Get bodyTypeNo() -> (bodyTypeNo:int=0)
    ) -> (ReturnValue:double)
   ) -> (Output_Get:double)
   Set VelocityScAcc_A(VelocityScAcc_A="0.8,0.8,0.8")
   > Set VelocityScAcc_B(VelocityScAcc_B="0.2,0.2,0.2")
   > Set VelocityScAcc_C(VelocityScAcc_C="0.1,0.1,0.1")
   > Set VelocityScAcc_D(VelocityScAcc_D="0.15,0.15,0.15")
   TryGetPawnOwner()
   float * float(A=@n12, B=2)
  ]}
]
variables [
 Backpack Pickaxe Rot {type=Rotator}
 BackpackRotate {type=Rotator}
 BackpackTranslation {type=Vector}
 BacpackInitPositionAdd {type=number}
 bodyTypeNo {type=integer}
 Delta Time X {type=number}
 DeltaTimeModifier {type=number, default=0.1}
 Gravity Override {type=Vector}
 Gravity Override Params {type=GravityOverrideParamsStruct}
 Head Down Rot {type=number}
 isM_Med {type=boolean}
 isMelee {type=number}
 packDBNOTransY {type=number}
 Paraсhuting {type=number}
 SimlationSpace {type=AnimPhysSimSpaceType}
 Spine Delta Transform from Ref Z {type=number}
 TargetRotX {type=number}
 VelocityScAcc_A {type=Vector}
 VelocityScAcc_B {type=Vector}
 VelocityScAcc_C {type=Vector}
 VelocityScAcc_D {type=Vector}
]
